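# Animation Blueprint for the TwiceBaked pickaxe, adding secondary motion to its wire and stone bones.
# AnimGraph chains sprung Anim Dynamics on six wire bones into additive-translation Modify Bone nodes for ten stones and one wire.
# Each update derives ten stone float vectors from a time-based sine and sets simulation space to World in front end, otherwise Component.
# Those variables show no links to the AnimGraph pins, which hold constants.
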AnimBlueprint /BRCosmetics/Weapons/FORT_Melee/Pickaxe_TwiceBaked/Meshes/TwiceBaked_Axe_AnimBP
target_skeleton=&/TwiceBaked_Axe_Skeleton
graphs [
 AnimGraph {nodes [
   AnimGraphNode_SubInput_0 = Input Pose()
   n1 = Anim Dynamics - Bone: dyn_wire_up_r!(
    SimulationSpace:AnimPhysSimSpaceType Enum=Component
    ComponentPose=Local To Component(LocalPose=@AnimGraphNode_SubInput_0)
    Alpha=0.7
    ExternalForce="0,0,0"
   ) -> (Pose) {
    angularDampingOverride=1
    angularSpringConstant=20000
    bAngularSpring=true
    bOverrideAngularDamping=true
    bUseGravityOverride=true
    gravityScale=0.35
    alphaBoolBlend {blendOption=Linear}
    boundBone {boneName=dyn_wire_up_r}
    componentLinearAccScale=<1 1 1>
    componentLinearVelScale=<1 1 1>
    gravityOverride {z=-980}
    physicsBodyDefinitions [{
      boundBone {boneName=dyn_wire_up_r}
      boxExtents=<20 35 35>
      constraintSetup {angularLimitsMax=<50 50 50>, angularLimitsMin=<-50 -50 -50>, angularTarget=<1 0 0>}
      localJointOffset=<2 0 0>
     }]
   }
   n2 = Anim Dynamics - Bone: dyn_wire_up_l!(SimulationSpace:AnimPhysSimSpaceType Enum=Component, ComponentPose=@n1, Alpha=0.7, ExternalForce="0,0,0") -> (Pose) {
    angularDampingOverride=1
    angularSpringConstant=20000
    bAngularSpring=true
    bOverrideAngularDamping=true
    bUseGravityOverride=true
    gravityScale=0.35
    alphaBoolBlend {blendOption=Linear}
    boundBone {boneName=dyn_wire_up_l}
    componentLinearAccScale=<1 1 1>
    componentLinearVelScale=<1 1 1>
    gravityOverride {z=-980}
    physicsBodyDefinitions [{
      boundBone {boneName=dyn_wire_up_l}
      boxExtents=<20 35 35>
      constraintSetup {angularLimitsMax=<50 50 50>, angularLimitsMin=<-50 -50 -50>, angularTarget=<1 0 0>}
      localJointOffset=<4 0 0>
     }]
   }
   n3 = Anim Dynamics - Bone: dyn_wire_mid_r!(SimulationSpace:AnimPhysSimSpaceType Enum=Component, ComponentPose=@n2, Alpha=0.6, ExternalForce="0,0,0") -> (Pose) {
    angularDampingOverride=1
    angularSpringConstant=20000
    bAngularSpring=true
    bOverrideAngularDamping=true
    bUseGravityOverride=true
    gravityScale=0.35
    alphaBoolBlend {blendOption=Linear}
    boundBone {boneName=dyn_wire_mid_r}
    componentLinearAccScale=<1 1 1>
    componentLinearVelScale=<1 1 1>
    gravityOverride {z=-980}
    physicsBodyDefinitions [{
      boundBone {boneName=dyn_wire_mid_r}
      boxExtents=<30 45 45>
      constraintSetup {angularLimitsMax=<30 30 30>, angularLimitsMin=<-30 -30 -30>, angularTarget=<1 0 0>}
      localJointOffset=<4 0 0>
     }]
   }
   n4 = Anim Dynamics - Bone: dyn_wire_down_r!(SimulationSpace:AnimPhysSimSpaceType Enum=Component, ComponentPose=@n3, Alpha=0.8, ExternalForce="0,0,0") -> (Pose) {
    angularDampingOverride=1
    angularSpringConstant=20000
    bAngularSpring=true
    bOverrideAngularDamping=true
    bUseGravityOverride=true
    gravityScale=0.35
    alphaBoolBlend {blendOption=Linear}
    boundBone {boneName=dyn_wire_down_r}
    componentLinearAccScale=<1 1 1>
    componentLinearVelScale=<1 1 1>
    gravityOverride {z=-980}
    physicsBodyDefinitions [{
      boundBone {boneName=dyn_wire_down_r}
      boxExtents=<30 45 45>
      constraintSetup {angularLimitsMax=<25 25 50>, angularLimitsMin=<-25 -25 -50>, angularTarget=<1 0 0>}
      localJointOffset=<4 0 0>
     }]
   }
   n5 = Anim Dynamics - Bone: dyn_wire_down_l!(SimulationSpace:AnimPhysSimSpaceType Enum=Component, ComponentPose=@n4, Alpha=0.6, ExternalForce="0,0,0") -> (Pose) {
    angularDampingOverride=1
    angularSpringConstant=20000
    bAngularSpring=true
    bOverrideAngularDamping=true
    bUseGravityOverride=true
    gravityScale=0.35
    alphaBoolBlend {blendOption=Linear}
    boundBone {boneName=dyn_wire_down_l}
    componentLinearAccScale=<1 1 1>
    componentLinearVelScale=<1 1 1>
    gravityOverride {z=-980}
    physicsBodyDefinitions [{
      boundBone {boneName=dyn_wire_down_l}
      boxExtents=<30 45 45>
      constraintSetup {angularLimitsMax=<25 25 50>, angularLimitsMin=<-25 -25 -50>, angularTarget=<1 0 0>}
      localJointOffset=<4 0 0>
     }]
   }
   n6 = Anim Dynamics - Bone: dyn_wire_mid_l!(SimulationSpace:AnimPhysSimSpaceType Enum=Component, ComponentPose=@n5, Alpha=0.6, ExternalForce="0,0,0") -> (Pose) {
    angularDampingOverride=1
    angularSpringConstant=20000
    bAngularSpring=true
    bOverrideAngularDamping=true
    bUseGravityOverride=true
    gravityScale=0.35
    alphaBoolBlend {blendOption=Linear}
    boundBone {boneName=dyn_wire_mid_l}
    componentLinearAccScale=<1 1 1>
    componentLinearVelScale=<1 1 1>
    gravityOverride {z=-980}
    physicsBodyDefinitions [{
      boundBone {boneName=dyn_wire_mid_l}
      boxExtents=<30 40 40>
      constraintSetup {angularLimitsMax=<0 50 50>, angularLimitsMin=<0 -50 -50>, angularTarget=<1 0 0>}
      localJointOffset=<4 0 0>
     }]
   }
   n7 = "Transform (Modify) Bone - Bone: base_stone_01_c"(ComponentPose=@n6, Alpha=1, Translation="0,0,0", Rotation="0,0,0", Scale="1,1,1") {
    rotationMode=BMM_Additive
    translationMode=BMM_Additive
    translationSpace=BCS_BoneSpace
    alphaBoolBlend {blendOption=Linear}
    boneToModify {boneName=base_stone_01_c}
   }
   n8 = "Transform (Modify) Bone - Bone: base_stone_02_c"(ComponentPose=@n7, Alpha=1, Translation="0,0,0", Rotation="0,0,0", Scale="1,1,1") {translationMode=BMM_Additive, translationSpace=BCS_BoneSpace, alphaBoolBlend {blendOption=Linear}, boneToModify {boneName=base_stone_02_c}}
   n9 = "Transform (Modify) Bone - Bone: base_stone_03_c"(ComponentPose=@n8, Alpha=1, Translation="0,0,0", Rotation="0,0,0", Scale="1,1,1") {translationMode=BMM_Additive, translationSpace=BCS_BoneSpace, alphaBoolBlend {blendOption=Linear}, boneToModify {boneName=base_stone_03_c}}
   n10 = "Transform (Modify) Bone - Bone: base_stone_04_c"(ComponentPose=@n9, Alpha=1, Translation="0,0,0", Rotation="0,0,0", Scale="1,1,1") {translationMode=BMM_Additive, translationSpace=BCS_BoneSpace, alphaBoolBlend {blendOption=Linear}, boneToModify {boneName=base_stone_04_c}}
   n11 = "Transform (Modify) Bone - Bone: dyn_stone_05_c"(ComponentPose=@n10, Alpha=1, Translation="0,0,0", Rotation="0,0,0", Scale="1,1,1") {translationMode=BMM_Additive, translationSpace=BCS_BoneSpace, alphaBoolBlend {blendOption=Linear}, boneToModify {boneName=dyn_stone_05_c}}
   n12 = "Transform (Modify) Bone - Bone: base_stone_06_c"(ComponentPose=@n11, Alpha=1, Translation="0,0,0", Rotation="0,0,0", Scale="1,1,1") {translationMode=BMM_Additive, translationSpace=BCS_BoneSpace, alphaBoolBlend {blendOption=Linear}, boneToModify {boneName=base_stone_06_c}}
   n13 = "Transform (Modify) Bone - Bone: base_stone_07_c"(ComponentPose=@n12, Alpha=1, Translation="0,0,0", Rotation="0,0,0", Scale="1,1,1") {translationMode=BMM_Additive, translationSpace=BCS_BoneSpace, alphaBoolBlend {blendOption=Linear}, boneToModify {boneName=base_stone_07_c}}
   n14 = "Transform (Modify) Bone - Bone: base_stone_08_c"(ComponentPose=@n13, Alpha=1, Translation="0,0,0", Rotation="0,0,0", Scale="1,1,1") {translationMode=BMM_Additive, translationSpace=BCS_BoneSpace, alphaBoolBlend {blendOption=Linear}, boneToModify {boneName=base_stone_08_c}}
   n15 = "Transform (Modify) Bone - Bone: base_stone_09_c"(ComponentPose=@n14, Alpha=1, Translation="0,0,0", Rotation="0,0,0", Scale="1,1,1") {translationMode=BMM_Additive, translationSpace=BCS_BoneSpace, alphaBoolBlend {blendOption=Linear}, boneToModify {boneName=base_stone_09_c}}
   n16 = "Transform (Modify) Bone - Bone: base_stone_10_c"(ComponentPose=@n15, Alpha=1, Translation="0,0,0", Rotation="0,0,0", Scale="1,1,1") {translationMode=BMM_Additive, translationSpace=BCS_BoneSpace, alphaBoolBlend {blendOption=Linear}, boneToModify {boneName=base_stone_10_c}}
   n17 = "Transform (Modify) Bone - Bone: base_wire_mid_l"(ComponentPose=@n16, Alpha=1, Translation="0,0,0", Rotation="0,0,0", Scale="1,1,1") {rotationMode=BMM_Additive, translationMode=BMM_Additive, alphaBoolBlend {blendOption=Linear}, boneToModify {boneName=base_wire_mid_l}}
   Output Pose(Result=Component To Local(ComponentPose=@n17)) {layerGroup=None}
  ]}
 EventGraph {nodes [
   Event BlueprintUpdateAnimation()
   > n1 = Sequence!(execute) -> (then_0, then_1, then_2:Exec)
   n2 = "Sin (Radians)"(A=float * float(A=GetRealTimeSeconds(), B=1.5))
   > Set Stone_01_Float(Stone_01_Float:Vector=MakeVector(X=float * float(A=@n2, B=0.3) {is=title K2Node_PromotableOperator}, Y=float * float(A=@n2, B=1), Z=float * float(A=@n2, B=-0.2))) -> (Output_Get:Vector)
   > Set Stone_02_Float(Stone_02_Float:Vector=MakeVector(X=float * float(A=@n2, B=0.7), Y=float * float(A=@n2, B=-0.9), Z=float * float(A=@n2, B=0.5))) -> (Output_Get:Vector)
   > Set Stone_03_Float(Stone_03_Float:Vector=MakeVector(X=float * float(A=@n2, B=-0.2), Y=float * float(A=@n2, B=0.5), Z=float * float(A=@n2, B=0.3))) -> (Output_Get:Vector)
   > Set Stone_04_Float(Stone_04_Float:Vector=MakeVector(X=float * float(A=@n2, B=-0.2), Y=float * float(A=@n2, B=-0.5), Z=float * float(A=@n2, B=-0.8))) -> (Output_Get:Vector)
   n3 = float * float(A=@n2, B=-0.2)
   n4 = float / float(A=@n2, B=1)
   > Set Stone_05_Float(Stone_05_Float:Vector=MakeVector(X=float * float(A=@n3, B=-0.2), Y=float * float(A=@n2, B=0.5), Z=float * float(A=@n4, B=-0.2))) -> (Output_Get:Vector)
   > Set Stone_06_Float(Stone_06_Float:Vector=MakeVector(X=float * float(A=@n2, B=0.2), Y=float * float(A=@n2, B=0.5), Z=float * float(A=@n2, B=-0.5))) -> (Output_Get:Vector)
   > Set Stone_07_Float(Stone_07_Float:Vector=MakeVector(X=float * float(A=@n2, B=-0.2), Y=float * float(A=@n2, B=1), Z=float * float(A=@n2, B=-0.3))) -> (Output_Get:Vector)
   > Set Stone_08_Float(Stone_08_Float:Vector=MakeVector(X=float * float(A=@n2, B=-0.2), Y=float * float(A=@n2, B=0.5), Z=float * float(A=@n2, B=0.5))) -> (Output_Get:Vector)
   > Set Stone_09_Float(Stone_09_Float:Vector=MakeVector(X=float * float(A=@n2, B=-0.2), Y=float * float(A=@n2, B=1), Z=float * float(A=@n2, B=-0.3))) -> (Output_Get:Vector)
   > Set Stone_10_Float(Stone_10_Float:Vector=MakeVector(X=float * float(A=@n2, B=-0.3), Y=float * float(A=@n2, B=1), Z=float * float(A=@n2, B=0.2))) -> (Output_Get:Vector)
   n5 = Branch(execute=@n1.then_1, Condition=Get bIsInFrontEnd())
   > Set SimlationSpace(SimlationSpace:AnimPhysSimSpaceType Enum=World) -> (Output_Get:AnimPhysSimSpaceType Enum)
   Set SimlationSpace(execute=@n5.else, SimlationSpace=Component)
   TryGetPawnOwner()
  ]}
]
variables [
 Is_Jump {type=number}
 jumpAlpha {type=number}
 SimlationSpace {type=AnimPhysSimSpaceType}
 Stone_01_Float {type=Vector, default=<-0.6868 0.9855 0.2194>}
 Stone_02_Float {type=Vector, default=<-0.6868 -0.5913 -0.351>}
 Stone_03_Float {type=Vector, default=<-0.6868 1.1826 -0.2194>}
 Stone_04_Float {type=Vector, default=<-0.6868 1.3797 0.2632>}
 Stone_05_Float {type=Vector, default=<-0.6868 0.3942 -0.3948>}
 Stone_06_Float {type=Vector, default=<-0.6868 1.1826 0.4387>}
 Stone_07_Float {type=Vector, default=<-0.6868 -0.7884 0.1755>}
 Stone_08_Float {type=Vector, default=<-0.6868 -0.9855 -0.3071>}
 Stone_09_Float {type=Vector, default=<-0.6868 0.9855 0.2194>}
 Stone_10_Float {type=Vector, default=<-0.6868 1.0841 -0.2632>}
]
